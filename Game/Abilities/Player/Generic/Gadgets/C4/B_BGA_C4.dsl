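; C4 gadget projectile that bounces off pawns and other C4, sticks where it lands and shows a team-colored beacon.
; On impact it stops, attaches to doors or tagged actors, and watches its support, rebinding to nearby buildings if that is lost.
; Trigger, damage, owner death, support loss or tagged overlaps detonate it.
; The server sends explosion gameplay events and destroys the actor, while RepNotify on bExploded plays effects.
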(event EventBeginPlay
  (bind _returnvalue (Game|GetPlayerController 0))
  (bind _returnvalue_2 (Pawn|GetControlledPawn _returnvalue))
  (bind _returnvalue_3 (Game|GetInstigator))
  (bind _asfort_player_pawn (Utilities|Casting|CastToFortPlayerPawn (Game|GetInstigator)))
  (EventDispatchers|AssignOnDied _asfort_player_pawn (delegate Custom|ExplodeOnOwnerDeath))
  (Utilities|FlowControl|Sequence
    (:then_0
      (Networking|CanExecuteCosmeticEvents
        (:True
          (Effects|Components|ParticleSystem|SetColorParameter (Variables|Default|GetP_ProximityMine_Beacon_01) "TeamColor" (select (Team|OnSameTeam _returnvalue_2 _asfort_player_pawn) "(R=0.000000,G=0.200000,B=0.800000,A=1.000000)" "(R=1.000000,G=0.000000,B=0.000000,A=1.000000)"))
          (Rendering|Material|SetVectorParameterValueonMaterials (Variables|Default|GetSM_C4_Explosive) "Emissive Multiplier" (Math|Conversions|ToVector(LinearColor) (select (Team|OnSameTeam _returnvalue_2 _asfort_player_pawn) "(R=0.000000,G=0.200000,B=0.800000,A=1.000000)" "(R=1.000000,G=0.000000,B=0.000000,A=1.000000)"))))))
    (:then_1
      ;[1 1 1] On Load
      (Variables|/Script/Engine.ProjectileMovementComponent|SetBounciness 0.2 (Variables|Default|GetProjectileMovement1))
      (if (Utilities|IsValid _returnvalue_3 :!class /Script/Engine.KismetSystemLibrary)
        (CallFunction|SetTeam (Team|GetActorTeam _returnvalue_3))))))

(event Custom|ExplodeOnOwnerDeath
  (params (DamagedActor :type Actor%20Object%20Reference) (Damage :type Float%20%28single-precision%29) (InstigatedBy :type Controller%20Object%20Reference) (DamageCauser :type Actor%20Object%20Reference) (HitLocation :type Vector) (FHitComponent :type Primitive%20Component%20Object%20Reference) (BoneName :type Name) (Momentum :type Vector))
  (bind _player_owner (Variables|Default|GetPlayer_Owner))
  (bind _returnvalue (Pawn|GetController _player_owner))
  (bind _asfort_player_controller (Utilities|Casting|CastToFortPlayerController _returnvalue))
  (bind _c4itemdef (Variables|Default|GetC4ItemDef))
  ;[1 1 1] when your owner dies, see if you are supposed to destroy yourself
  (if (Variables|Default|GetbDestroyOnDeath)
    (CallFunction|DamageExplode _player_owner)
    (if (== (Inventory|GetItemQuantityOnPlayer _asfort_player_controller _c4itemdef) 0)
      (Inventory|RemoveItemFromPlayer :PlayerController _asfort_player_controller :ItemDefinition _c4itemdef :AmountToRemove -1 :bForceRemoval true :bSkipVerbRemoval false))))

(event Custom|SetExplosion
  (params (Explosion%20Level :type Integer) (Activation_Radius :type Float%20%28double-precision%29) (PlayerPawn :type Fort%20Player%20Pawn%20Object%20Reference))
  (Variables|Default|SetActivation_Radius Activation_Radius)
  (Variables|Default|SetPlayer_Owner PlayerPawn)
  (Variables|Default|SetExplosionLevel Explosion%20Level)
  ;[0.1261 0.1093 1] Scaling the explosion
  (Variables|Default|SetEventExplosion (Variables|Default|GetExplosionRangeDefaultTag)))

(event Custom|SetProjectile
  (params (Init_Speed :type Float%20%28double-precision%29) (Gravity :type Float%20%28double-precision%29) (Activation_Delay :type Float%20%28double-precision%29))
  (bind _projectilemovement1 (Variables|Default|GetProjectileMovement1))
  ;[1 0.03636 0] Sets Prj movement from GA  (Not getting values!!!!)
  (Variables|/Script/Engine.ProjectileMovementComponent|SetInitialSpeed Init_Speed _projectilemovement1)
  (Variables|/Script/Engine.ProjectileMovementComponent|SetProjectileGravityScale Gravity _projectilemovement1)
  (Variables|Default|SetActivation_Delay Activation_Delay))

(event Collision|EventHit
  (params (MyComp :type Primitive%20Component%20Object%20Reference) (Other :type Actor%20Object%20Reference) (OtherComp :type Primitive%20Component%20Object%20Reference) (bSelfMoved :type Boolean) (HitLocation :type Vector :default "0, 0, 0") (HitNormal :type Vector :default "0, 0, 0") (NormalImpulse :type Vector :default "0, 0, 0") (Hit :type Hit%20Result%20Structure%20%28by%20ref%29))
  (bind _impact_loc (Variables|Default|GetImpact_Loc))
  (bind _land_sound (Variables|Default|GetLand_Sound))
  (bind _sm_c4_explosive (Variables|Default|GetSM_C4_Explosive))
  ;[1 1 1] When this actor hits something, if its not a pawn, attach to it!
  (Utilities|Casting|CastToB_BGA_C4 Other
    (:then
      (CallFunction|EventFakeBounce))
    (:CastFailed
      (Utilities|FlowControl|DoOnce false
        (:Completed
          (bind _returnvalue (GameplayTags|HasAnyMatchingGameplayTags(Message) Other (Variables|Default|GetTC_ActorTagsThatShouldHaveCustomLogicOnOverlapOrHit)))
          (if _returnvalue
            (Transformation|AttachActorToComponent OtherComp "None" "KeepWorld" "KeepWorld" "KeepWorld" false)
            (else
              (CallFunction|HandleHitBuildingWall Other OtherComp)))
          (Variables|Default|SetImpact_Loc HitLocation)
          (Variables|Default|SetHitNormal HitNormal)
          (Game|Components|ProjectileMovement|StopSimulating (Variables|Default|GetProjectileMovement1) Hit)
          ;[1 1 1] Binds Event for when Building piece dies
          (bind _asbuilding_actor (Utilities|Casting|CastToBuildingActor Other)
            (:then
              (EventDispatchers|BindEventtoOnDied _asbuilding_actor (EventDispatchers|CreateEvent :!event Custom|BaseDied)))
            (:CastFailed
              (Game|BindEventtoOnDestroyed Other (EventDispatchers|CreateEvent :!event Custom|BaseDestroyed))))
          ;; MinePlaced
          (Variables|Default|SetRotationInAir (Transformation|GetActorRotation self))
          (Transformation|SetActorRotation (Math|Rotator|MakeRotfromZ (Variables|Default|GetHitNormal)))
          (Transformation|AddRelativeRotation _sm_c4_explosive (Math|Rotator|MakeRotator 0.0 0.0 (+ (- (.yaw (Variables|Default|GetRotationInAir)) 9e+01) (Math|Random|RandomFloatinRange -2e+01 2e+01))) false false)
          (Transformation|AddRelativeLocation _sm_c4_explosive "0, 0, 0" false false)
          (Collision|SetCollisionEnabled (Variables|Default|GetRoot_Sphere_Collision) "NoCollision")
          (Networking|CanExecuteCosmeticEvents
            (:True
              (Audio|PlaySoundatLocation _land_sound _impact_loc)
              (Fortnite|ShowSoundIndicatorLocalforTeam :Location _impact_loc :MaxAudibleDistance (Variables|/Script/Engine.SoundBase|GetMaxDistance _land_sound) :IndicatorType "Gunshot" :Team (Team|GetActorTeam self) :Affiliations (Utilities|Array|MakeArray "Hostile" "Neutral") :Tint "(R=0.000000,G=0.000000,B=0.000000,A=0.000000)" :AttenuationMaxRangeMultiplier 1.0 :AttenuationIntensityMultiplier 1.0)
              (Effects|SpawnEmitteratLocation "/Game/Athena/Items/Consumables/C4/P_Athena_C4_Land_Dust.P_Athena_C4_Land_Dust" (Variables|Default|GetImpact_Loc) (Math|Conversions|RotationFromXVector (Variables|Default|GetHitNormal)))
              (Components|Activation|SetActive (Variables|Default|GetP_ProximityMine_Beacon_01) true false))))))))

(event Custom|BaseDied
  (params (DamagedActor :type Actor%20Object%20Reference) (Damage :type Float%20%28single-precision%29) (InstigatedBy :type Controller%20Object%20Reference) (DamageCauser :type Actor%20Object%20Reference) (HitLocation :type Vector) (FHitComponent :type Primitive%20Component%20Object%20Reference) (BoneName :type Name) (Momentum :type Vector))
  (bind _returnvalue (Building|GetOwningPlayerState))
  (bind _returnvalue_2 (Building|GetOwningPlayerState))
  (bind (_outactors _returnvalue_3) (Collision|SphereOverlapActors (Transformation|GetActorLocation) 16.0 (Utilities|Array|MakeArray "ObjectTypeQuery1") "/Script/FortniteGame.BuildingActor" (Utilities|Array|MakeArray self)))
  (if _returnvalue_3
    (bind _output_get (Variables|Default|SetFoundBuildingOnDied _outactors))
    (CallFunction|HandleWasAttachedtoWall (Utilities|Array|Get(acopy) _output_get))
    (bind _asbuilding_actor (Utilities|Casting|CastToBuildingActor (Utilities|Array|Get(acopy) _output_get))
      (:then
        (EventDispatchers|BindEventtoOnDied _asbuilding_actor (EventDispatchers|CreateEvent :!event Custom|BaseDied)))
      (:CastFailed
        (Game|BindEventtoOnDestroyed _asbuilding_actor (EventDispatchers|CreateEvent :!event Custom|BaseDestroyed))))
    (else
      ;[1 1 1] variable delay before explode because what you were attached to was destroyed
      (Utilities|FlowControl|Delay (Math|Random|RandomFloatinRange 0.0 0.6)) ; .25
      (Utilities|FlowControl|DoOnce false
        (:Completed
          ;[0.8855 0.04538 1]
          (Variables|Default|SetIsDead? true)
          (Networking|SwitchHasAuthority
            (:Authority
              (Utilities|IsValid _returnvalue_2
                (:Is%20Valid
                  (Ability|SendGameplayEventtoActor self (Variables|Default|GetEventExplosion) (Utilities|Struct|MakeGameplayEventData (Variables|Default|GetEventExplosion) _returnvalue_2 _returnvalue_2 self))
                  (Variables|Default|SetbExploded true)
                  (Utilities|IsValid _returnvalue
                    (:Is%20Valid
                      (Ability|SendGameplayEventtoActor _returnvalue (Variables|Default|GetEventExplosion) (Utilities|Struct|MakeGameplayEventData (Variables|Default|GetEventExplosion) _returnvalue _returnvalue self)))
                    (:Is%20Not%20Valid
                      _))
                  (Actor|DestroyActor))
                (:Is%20Not%20Valid
                  _)))) ; If touched or damage\\ Only on server will it trigger explosion
        )))))

(event Custom|BaseDestroyed
  (params (DestroyedActor :type Actor%20Object%20Reference))
  (bind _returnvalue (Building|GetOwningPlayerState))
  (bind _returnvalue_2 (Building|GetOwningPlayerState))
  (bind (_outactors _returnvalue_3) (Collision|SphereOverlapActors (Transformation|GetActorLocation) 16.0 (Utilities|Array|MakeArray "ObjectTypeQuery1") "/Script/FortniteGame.BuildingActor" (Utilities|Array|MakeArray self)))
  (if _returnvalue_3
    (bind _output_get (Variables|Default|SetFoundBuildingOnDied _outactors))
    (CallFunction|HandleWasAttachedtoWall (Utilities|Array|Get(acopy) _output_get))
    (bind _asbuilding_actor (Utilities|Casting|CastToBuildingActor (Utilities|Array|Get(acopy) _output_get))
      (:then
        (EventDispatchers|BindEventtoOnDied _asbuilding_actor (EventDispatchers|CreateEvent :!event Custom|BaseDied)))
      (:CastFailed
        (Game|BindEventtoOnDestroyed _asbuilding_actor (EventDispatchers|CreateEvent :!event Custom|BaseDestroyed))))
    (else
      ;[1 1 1] variable delay before explode because what you were attached to was destroyed
      (Utilities|FlowControl|Delay (Math|Random|RandomFloatinRange 0.0 0.6)) ; .25
      (Utilities|FlowControl|DoOnce false
        (:Completed
          ;[0.8855 0.04538 1]
          (Variables|Default|SetIsDead? true)
          (Networking|SwitchHasAuthority
            (:Authority
              (Utilities|IsValid _returnvalue_2
                (:Is%20Valid
                  (Ability|SendGameplayEventtoActor self (Variables|Default|GetEventExplosion) (Utilities|Struct|MakeGameplayEventData (Variables|Default|GetEventExplosion) _returnvalue_2 _returnvalue_2 self))
                  (Variables|Default|SetbExploded true)
                  (Utilities|IsValid _returnvalue
                    (:Is%20Valid
                      (Ability|SendGameplayEventtoActor _returnvalue (Variables|Default|GetEventExplosion) (Utilities|Struct|MakeGameplayEventData (Variables|Default|GetEventExplosion) _returnvalue _returnvalue self)))
                    (:Is%20Not%20Valid
                      _))
                  (Actor|DestroyActor))
                (:Is%20Not%20Valid
                  _)))) ; If touched or damage\\ Only on server will it trigger explosion
        )))))

(event Custom|ExplodeNow
  (bind _returnvalue (Building|GetOwningPlayerState))
  (bind _returnvalue_2 (Building|GetOwningPlayerState))
  (Utilities|FlowControl|DoOnce false
    (:Completed
      ;[0.8855 0.04538 1]
      (Variables|Default|SetIsDead? true)
      (Networking|SwitchHasAuthority
        (:Authority
          (Utilities|IsValid _returnvalue_2
            (:Is%20Valid
              (Ability|SendGameplayEventtoActor self (Variables|Default|GetEventExplosion) (Utilities|Struct|MakeGameplayEventData (Variables|Default|GetEventExplosion) _returnvalue_2 _returnvalue_2 self))
              (Variables|Default|SetbExploded true)
              (Utilities|IsValid _returnvalue
                (:Is%20Valid
                  (Ability|SendGameplayEventtoActor _returnvalue (Variables|Default|GetEventExplosion) (Utilities|Struct|MakeGameplayEventData (Variables|Default|GetEventExplosion) _returnvalue _returnvalue self)))
                (:Is%20Not%20Valid
                  _))
              (Actor|DestroyActor))
            (:Is%20Not%20Valid
              _)))) ; If touched or damage\\ Only on server will it trigger explosion
    )))

(event OnComponentBeginOverlap(Root_Sphere_Collision)
  (params (OverlappedComponent :type Primitive%20Component%20Object%20Reference) (OtherActor :type Actor%20Object%20Reference) (OtherComp :type Primitive%20Component%20Object%20Reference) (OtherBodyIndex :type Integer) (bFromSweep :type Boolean) (SweepResult :type Hit%20Result%20Structure%20%28by%20ref%29))
  ;[1 1 1] Fake bounce off of pawns
  (bind _asfort_pawn (Utilities|Casting|CastToFortPawn OtherActor)
    (:then
      (if (Utilities|NotEqual(Object) _asfort_pawn (Actor|GetOwner))
        (CallFunction|EventFakeBounce)))
    (:CastFailed
      (bind _returnvalue (GameplayTags|HasAnyMatchingGameplayTags(Message) OtherActor (Variables|Default|GetTC_ActorTagsThatShouldHaveCustomLogicOnOverlapOrHit)))
      (if _returnvalue
        (CallFunction|ExplodeNow)))))

(event Custom|DamageExplode
  (params (DamageOwner :type Fort%20Player%20Pawn%20Object%20Reference))
  (bind _returnvalue (Building|GetOwningPlayerState))
  (bind _returnvalue_2 (Building|GetOwningPlayerState))
  ;[1 0.09874 0] Damage caused this to explode
  (Variables|Default|SetPlayer_Owner DamageOwner)
  ;[1 1 1] variable delay before explode if you were damaged
  (Utilities|FlowControl|Delay (Math|Random|RandomFloatinRange 0.0 0.2)) ; .25
  (Utilities|FlowControl|DoOnce false
    (:Completed
      ;[0.8855 0.04538 1]
      (Variables|Default|SetIsDead? true)
      (Networking|SwitchHasAuthority
        (:Authority
          (Utilities|IsValid _returnvalue_2
            (:Is%20Valid
              (Ability|SendGameplayEventtoActor self (Variables|Default|GetEventExplosion) (Utilities|Struct|MakeGameplayEventData (Variables|Default|GetEventExplosion) _returnvalue_2 _returnvalue_2 self))
              (Variables|Default|SetbExploded true)
              (Utilities|IsValid _returnvalue
                (:Is%20Valid
                  (Ability|SendGameplayEventtoActor _returnvalue (Variables|Default|GetEventExplosion) (Utilities|Struct|MakeGameplayEventData (Variables|Default|GetEventExplosion) _returnvalue _returnvalue self)))
                (:Is%20Not%20Valid
                  _))
              (Actor|DestroyActor))
            (:Is%20Not%20Valid
              _)))) ; If touched or damage\\ Only on server will it trigger explosion
    )))

(event Custom|Event%20Fake%20Bounce
  (bind (_returnvalue_x _returnvalue_y _returnvalue_z) (Math|Vector|BreakVector (* (Variables|/Script/Engine.MovementComponent|GetVelocity (Variables|Default|GetProjectileMovement1)) -1.0)))
  (bind _projectilemovement1 (Variables|Default|GetProjectileMovement1))
  ;[1 1 1] Fakes a bounce backwards.
  (Variables|/Script/Engine.ProjectileMovementComponent|SetProjectileGravityScale (Variables|Default|GetGravityScaleForBounce) _projectilemovement1)
  (Variables|/Script/Engine.MovementComponent|SetVelocity (* (Math|Vector|MakeVector _returnvalue_x _returnvalue_y (Math|Float|Absolute(Float) _returnvalue_z)) 0.25) (Variables|Default|GetProjectileMovement1)))

(event Custom|TriggeredExplosion
  (params (DelayForChainExplode :type Float%20%28double-precision%29))
  (bind _returnvalue (Building|GetOwningPlayerState))
  (bind _returnvalue_2 (Building|GetOwningPlayerState))
  (Utilities|FlowControl|DoOnce false
    (:Completed
      (Utilities|FlowControl|Delay DelayForChainExplode)
      (Audio|PlaySoundatLocation (Variables|Default|GetPreExplosion_Sound) (Transformation|GetWorldLocation (Variables|Default|GetSM_C4_Explosive)))
      ;[1 1 1] standard delay before explode when triggered
      (Utilities|FlowControl|Delay (Variables|Default|GetDefaultDelay)) ; .25
      (Utilities|FlowControl|DoOnce false
        (:Completed
          ;[0.8855 0.04538 1]
          (Variables|Default|SetIsDead? true)
          (Networking|SwitchHasAuthority
            (:Authority
              (Utilities|IsValid _returnvalue_2
                (:Is%20Valid
                  (Ability|SendGameplayEventtoActor self (Variables|Default|GetEventExplosion) (Utilities|Struct|MakeGameplayEventData (Variables|Default|GetEventExplosion) _returnvalue_2 _returnvalue_2 self))
                  (Variables|Default|SetbExploded true)
                  (Utilities|IsValid _returnvalue
                    (:Is%20Valid
                      (Ability|SendGameplayEventtoActor _returnvalue (Variables|Default|GetEventExplosion) (Utilities|Struct|MakeGameplayEventData (Variables|Default|GetEventExplosion) _returnvalue _returnvalue self)))
                    (:Is%20Not%20Valid
                      _))
                  (Actor|DestroyActor))
                (:Is%20Not%20Valid
                  _)))) ; If touched or damage\\ Only on server will it trigger explosion
        )))))

(event EventOnDamageServer
  (params (Damage :type Float%20%28single-precision%29) (DamageTags :type Gameplay%20Tag%20Container%20Structure%20%28by%20ref%29) (Momentum :type Vector :default "0, 0, 0") (HitInfo :type Hit%20Result%20Structure%20%28by%20ref%29) (InstigatedBy :type Controller%20Object%20Reference) (DamageCauser :type Actor%20Object%20Reference) (EffectContext :type Gameplay%20Effect%20Context%20Handle%20Structure))
  ;[1 1 1] On Damage switch the owner to the damage dealer and explode it
  (Networking|SwitchHasAuthority
    (:Authority
      (bind _asfort_player_pawn (Utilities|Casting|CastToFortPlayerPawn (Pawn|GetControlledPawn InstigatedBy)))
      (CallFunction|DamageExplode _asfort_player_pawn))))

(event Custom|PlayExplosionFX
  (bind _sm_c4_explosive (Variables|Default|GetSM_C4_Explosive))
  (bind _returnvalue (Transformation|GetWorldLocation _sm_c4_explosive))
  (bind _explosion_sound (Variables|Default|GetExplosion_Sound))
  ;[1 0.4852 0] Trigger effects on death
  (Rendering|SetVisibility (Variables|Default|GetSM_C4_Explosive) false false)
  (Audio|PlaySoundatLocation _explosion_sound _returnvalue) ; Explosion sound
  (Fortnite|ShowSoundIndicatorLocalforTeam :Location _returnvalue :MaxAudibleDistance (Variables|/Script/Engine.SoundBase|GetMaxDistance _explosion_sound) :IndicatorType "Gunshot" :Team (Team|GetActorTeam self) :Affiliations (Utilities|Array|MakeArray "Hostile" "Neutral") :Tint "(R=0.000000,G=0.000000,B=0.000000,A=0.000000)" :AttenuationMaxRangeMultiplier 1.0 :AttenuationIntensityMultiplier 1.0)
  (bind _returnvalue_2 (Niagara|SpawnSystematLocation (Variables|Default|GetNSExplosion) _returnvalue (Math|Rotator|MakeRotfromX (Variables|Default|GetHitNormal))))
  (Effects|Components|ParticleSystem|SetFloatParameter _returnvalue_2 "Global Scale" 1.05)
  (Effects|Components|ParticleSystem|SetFloatParameter _returnvalue_2 "Fire Intensity" 15.0)
  (Fortnite|PlayLocalForceFeedbackatLocationMulti (Variables|Default|GetExplosionForceFeedbackNear) (Variables|Default|GetExplosionForceFeedbackFar) (Transformation|GetActorLocation) (Variables|Default|GetExplosionRadius) (* (Variables|Default|GetExplosionRadius) 3) "Explosion")
  (Camera|PlayWorldCameraShake "/Game/Abilities/Player/Commando/Actives/FragGrenade/B_GrenadeCameraShake.B_GrenadeCameraShake_C" (Transformation|GetActorLocation) 200 800))

(fn HandleHitBuildingWall
  (params (Object :type Object%20Reference) (Comp :type Primitive%20Component%20Object%20Reference))
  (bind _asbuilding_wall (Utilities|Casting|CastToBuildingWall (Variables|GetObject)))
  (LocalVariables|SetLocalWall _asbuilding_wall)
  (bind _asstatic_mesh_component (Utilities|Casting|CastToStaticMeshComponent (Variables|GetComp)))
  (bind _returnvalue (BuildingWall|IsDoorComponent (LocalVariables|GetLocalWall) _asstatic_mesh_component))
  (if _returnvalue
    (Transformation|AttachActorToComponent (Variables|GetComp) "None" "KeepWorld" "KeepWorld" "KeepWorld" false)))

(fn HandleWasAttachedToWall
  (params (Object :type Object%20Reference))
  (bind (_outcomponents _returnvalue) (Collision|SphereOverlapComponents (Transformation|GetActorLocation) 16.0 (Utilities|Array|MakeArray "ObjectTypeQuery1") "/Script/Engine.StaticMeshComponent" (Utilities|Array|MakeArray self)))
  (if _returnvalue
    (CallFunction|HandleHitBuildingWall (Variables|GetObject) (Utilities|Array|Get(acopy) _outcomponents))))

(fn OnRep_bExploded
  (CallFunction|PlayExplosionFX))

(fn OnRep_IsDead?
  (Components|Activation|SetActive (Variables|Default|GetP_ProximityMine_Beacon_01) (not (Variables|Default|GetIsDead?)) false))

(fn ConstructionScript)
